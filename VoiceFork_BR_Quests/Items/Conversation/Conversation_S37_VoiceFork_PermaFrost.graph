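# Conversation database for the Season 37 VoiceFork PermaFrost questline, with NPC dialogue for BaskIsle, ArcanaAgate, OboeThorn and FluteLamp.
# Each offers and grants an urgent quest until its PermaFrost objective completes, or shows "mission already active" during another urgent quest.
# OboeThorn and FluteLamp add delivery entries gated on urgent-quest objectives.
# A DustyBun interrogation entry confesses once ArcanaAgate's objective completes, or feigns ignorance under a failure tag.
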ConversationDatabase /VoiceFork_BR_Quests/Items/Conversation/Conversation_S37_VoiceFork_PermaFrost
entry_points [
 {tag=Conversation.Entry.Season37.VoiceFork.BaskIsle, destinations [#1]}
 {tag=Conversation.Entry.Season37.VoiceFork.ArcanaAgate, destinations [#2]}
 {tag=Conversation.Entry.Season37.VoiceFork.ArcanaAgate.DustyBun, destinations [#3]}
 {tag=Conversation.Entry.Season37.VoiceFork.OboeThorn, destinations [#4]}
 {tag=Conversation.Entry.Season37.VoiceFork.OboeThorn.Deliver, destinations [#5]}
 {tag=Conversation.Entry.Season37.VoiceFork.FluteLamp, destinations [#6]}
 {tag=Conversation.Entry.Season37.VoiceFork.FluteLamp.Deliver, destinations [#7]}
]
nodes [
 ConversationEntryPointNode_6 {
  id=#7
  class=ConversationEntryPointNode
  outputs [#8]
  properties {entryTag {tagName=Conversation.Entry.Season37.VoiceFork.FluteLamp.Deliver}}
 }
 Speech_C_5 {
  id=#8
  class=Speech_C
  outputs [#9]
  properties {
   softTaskIcon=&T_UI_NPC_Conversation_Give
   generalConfig {defaultMessage=*sniff* medium rare... toasted bun... lots of ketchup..., defaultMetadataParameters [RageDebris {value=""}]}
  }
  sub_nodes [
   FortConversationRequirement_ControllerRequirement_7 {
    class=FortConversationRequirement_ControllerRequirement
    properties {
     nodeName=Require: Has Urgent Quest
     requirement {
      class=FortControllerRequirement_HasUrgentQuest
      properties {acceptedUrgentEventTags {gameplayTags [{tagName=SpecialEvent.Season37.Found.VoiceFork.FluteLamp}]}}
     }
    }
   }
   FortConversationRequirement_ControllerRequirement_8 {
    class=FortConversationRequirement_ControllerRequirement
    properties {
     nodeName=Require: Has Completed Quest Objective
     requirement {
      class=FortControllerRequirement_HasCompletedQuestObjective
      properties {
       bInverseCondition=true
       objectiveBackendName=Quest_S37_VoiceFork_PermaFrost_Q01_obj3
       questDefinition=/VoiceFork_BR_Quests/Items/QuestItems/PermaFrost/Quest_S37_VoiceFork_PermaFrost_Q01
      }
     }
    }
   }
   FortConversationRequirement_ControllerRequirement_9 {
    class=FortConversationRequirement_ControllerRequirement
    properties {
     nodeName=Require: Has Completed Quest Objective
     requirement {
      class=FortControllerRequirement_HasCompletedQuestObjective
      properties {
       objectiveBackendName=Quest_S37_VoiceFork_PermaFrost_FluteLamp_Q01_obj0
       questDefinition=/VoiceFork_BR_Quests/Items/QuestItems/PermaFrost/UrgentQuests/FluteLamp/Quest_S37_VoiceFork_PermaFrost_FluteLamp_Q01
      }
     }
    }
   }
   ConversationChoiceNode_16 {
    class=ConversationChoiceNode
    properties {
     defaultChoiceDisplayText=Deliver the burger
     choiceTags {gameplayTags [{tagName=Conversation.Choice.Category.MainQuestline}, {tagName=Conversation.Entry.Season37.VoiceFork.FluteLamp.Deliver}]}
    }
   }
   FortConversationRequirement_ControllerRequirement_49 {
    class=FortConversationRequirement_ControllerRequirement
    properties {
     nodeName=Require: Has Completed Quest Objective
     requirement {
      class=FortControllerRequirement_HasCompletedQuestObjective
      properties {
       objectiveBackendName=Quest_S37_VoiceFork_PermaFrost_FluteLamp_Q01_obj1
       questDefinition=&Quest_S37_VoiceFork_PermaFrost_FluteLamp_Q01
      }
     }
    }
   }
   FortConversationRequirement_ControllerRequirement_23 {
    class=FortConversationRequirement_ControllerRequirement
    properties {
     nodeName=Require: Has GameplayTag
     requirement {class=FortControllerRequirement_HasOwnedTag, properties {ownedTag {tagName=Conversation.Entry.Season37.VoiceFork.FluteLamp.Deliver}}}
    }
   }
  ]
 }
 Speech_C_22 {
  id=#9
  class=Speech_C
  outputs [#10]
  properties {
   softTaskIcon=&T_UI_NPC_MainQuest
   generalConfig {defaultMessage="All that's missing is the popping candy and the marshmallow fluff!", defaultMetadataParameters [RageDebris {value=""}]}
  }
 }
 Speech_C_23 {
  id=#10
  class=Speech_C
  outputs [#11]
  properties {
   softTaskIcon=&T_UI_NPC_MainQuest
   generalConfig {defaultMessage="Like, my compliments to the chef!", defaultMetadataParameters [RageDebris {value=""}]}
  }
 }
 ReturnToStart_C_6 {id=#11, class=ReturnToStart_C}
 ConversationEntryPointNode_5 {
  id=#6
  class=ConversationEntryPointNode
  outputs [#12, #13]
  properties {entryTag {tagName=Conversation.Entry.Season37.VoiceFork.FluteLamp}}
 }
 ShowUserDataDrivenCannotUseReason_C_3 {
  id=#13
  class=ShowUserDataDrivenCannotUseReason_C
  outputs []
  properties {
   cannotUseReason=JobAlreadyActive
   serviceBriefCollectionKeyOveride=Service.JobBoard.DisplayCannotUseReason
   serviceBriefConfigCollection=&/FoundQuestsServiceBriefCollectionConfig
   softTaskIcon=&Cancel_Normal
  }
  sub_nodes [
   ConversationChoiceNode_10 {
    class=ConversationChoiceNode
    properties {defaultChoiceDisplayText=MISSION ALREADY ACTIVE, choiceTags {gameplayTags [{tagName=Conversation.Choice.Category.Back}]}}
   }
   FortConversationRequirement_ControllerRequirement_15 {
    class=FortConversationRequirement_ControllerRequirement
    properties {nodeName=Require: Has Urgent Quest, requirement {class=FortControllerRequirement_HasUrgentQuest}}
   }
   FortConversationRequirement_ControllerRequirement_21 {
    class=FortConversationRequirement_ControllerRequirement
    properties {
     nodeName=Require: Has Urgent Quest
     requirement {
      class=FortControllerRequirement_HasUrgentQuest
      properties {bInverseCondition=true, acceptedUrgentEventTags {gameplayTags [{tagName=SpecialEvent.Season37.Found.VoiceFork.FluteLamp}]}}
     }
    }
   }
  ]
 }
 Speech_C_4 {
  id=#12
  class=Speech_C
  outputs [#14, #15]
  properties {
   softTaskIcon=&T_UI_NPC_MainQuest
   generalConfig {
    defaultMessage="This place is, like, FREAKY! I could totally use a burger to help calm me down."
    defaultMetadataParameters [RageDebris {value=""}]
   }
  }
  sub_nodes [
   ConversationChoiceNode_12 {
    class=ConversationChoiceNode
    properties {defaultChoiceDisplayText=Need help?, choiceTags {gameplayTags [{tagName=Conversation.Choice.Category.MainQuestline}]}}
   }
   FortConversationRequirement_ControllerRequirement_17 {
    class=FortConversationRequirement_ControllerRequirement
    properties {nodeName=Require: Has Urgent Quest, requirement {class=FortControllerRequirement_HasUrgentQuest, properties {bInverseCondition=true}}}
   }
   FortConversationRequirement_ControllerRequirement_18 {
    class=FortConversationRequirement_ControllerRequirement
    properties {
     nodeName=Require: Has Completed Quest Objective
     requirement {
      class=FortControllerRequirement_HasCompletedQuestObjective
      properties {
       bInverseCondition=true
       objectiveBackendName=Quest_S37_VoiceFork_PermaFrost_Q01_obj3
       questDefinition=&Quest_S37_VoiceFork_PermaFrost_Q01
      }
     }
    }
   }
  ]
 }
 Speech_C_21 {
  id=#15
  class=Speech_C
  outputs [#16]
  properties {
   softTaskIcon=&T_UI_NPC_Back
   generalConfig {defaultMessage="I'm, like, so hungry...", defaultMetadataParameters [RageDebris {value=""}]}
  }
  sub_nodes [ConversationChoiceNode_2 {
    class=ConversationChoiceNode
    properties {defaultChoiceDisplayText="Nah, not now", choiceTags {gameplayTags [{tagName=Conversation.Choice.CustomSortPriority.Highest}]}}
   }]
 }
 ReturnToStart_C_5 {id=#16, class=ReturnToStart_C}
 GrantStaticQuest_C_1 {
  id=#14
  class=GrantStaticQuest_C
  outputs [#17]
  properties {
   bAssignQuestOwner=true
   bSharedQuestLimitsEnforced=false
   bShouldOverrideSpeakerText=false
   nodeName=Grant Quest
   questToGrantPtr=&Quest_S37_VoiceFork_PermaFrost_FluteLamp_Q01
  }
  sub_nodes [ConversationChoiceNode_11 {
    class=ConversationChoiceNode
    properties {defaultChoiceDisplayText="Yes, count on me!", choiceTags {gameplayTags [{tagName=Conversation.Choice.CustomSortPriority.Lowest}]}}
   }]
 }
 Speech_C_20 {
  id=#17
  class=Speech_C
  outputs []
  properties {
   softTaskIcon=&T_UI_NPC_MainQuest
   generalConfig {defaultMessage=My mouth is watering already..., defaultMetadataParameters [RageDebris {value=""}]}
  }
 }
 ConversationEntryPointNode_9 {
  id=#5
  class=ConversationEntryPointNode
  outputs [#18]
  properties {entryTag {tagName=Conversation.Entry.Season37.VoiceFork.OboeThorn.Deliver}}
 }
 Speech_C_13 {
  id=#18
  class=Speech_C
  outputs [#19]
  properties {
   softTaskIcon=&T_UI_NPC_Conversation_Give
   generalConfig {defaultMessage="This is exactly what she needs. Thanks pal!", defaultMetadataParameters [RageDebris {value=""}]}
  }
  sub_nodes [
   FortConversationRequirement_ControllerRequirement_44 {
    class=FortConversationRequirement_ControllerRequirement
    properties {
     nodeName=Require: Has Urgent Quest
     requirement {
      class=FortControllerRequirement_HasUrgentQuest
      properties {acceptedUrgentEventTags {gameplayTags [{tagName=SpecialEvent.Season37.Found.VoiceFork.OboeThorn}]}}
     }
    }
   }
   FortConversationRequirement_ControllerRequirement_45 {
    class=FortConversationRequirement_ControllerRequirement
    properties {
     nodeName=Require: Has Completed Quest Objective
     requirement {
      class=FortControllerRequirement_HasCompletedQuestObjective
      properties {
       bInverseCondition=true
       objectiveBackendName=Quest_S37_VoiceFork_PermaFrost_Q01_obj0
       questDefinition=&Quest_S37_VoiceFork_PermaFrost_Q01
      }
     }
    }
   }
   FortConversationRequirement_ControllerRequirement_14 {
    class=FortConversationRequirement_ControllerRequirement
    properties {
     nodeName=Require: Has Completed Quest Objective
     requirement {
      class=FortControllerRequirement_HasCompletedQuestObjective
      properties {
       objectiveBackendName=Quest_S37_VoiceFork_PermaFrost_OboeThorn_Q01_obj0
       questDefinition=/VoiceFork_BR_Quests/Items/QuestItems/PermaFrost/UrgentQuests/OboeThorn/Quest_S37_VoiceFork_PermaFrost_OboeThorn_Q01
      }
     }
    }
   }
   FortConversationRequirement_ControllerRequirement_48 {
    class=FortConversationRequirement_ControllerRequirement
    properties {
     nodeName=Require: Has Completed Quest Objective
     requirement {
      class=FortControllerRequirement_HasCompletedQuestObjective
      properties {
       objectiveBackendName=Quest_S37_VoiceFork_PermaFrost_OboeThorn_Q01_obj1
       questDefinition=&Quest_S37_VoiceFork_PermaFrost_OboeThorn_Q01
      }
     }
    }
   }
   ConversationChoiceNode_31 {
    class=ConversationChoiceNode
    properties {
     defaultChoiceDisplayText=Deliver items
     choiceTags {gameplayTags [{tagName=Conversation.Choice.Category.MainQuestline}, {tagName=Conversation.Entry.Season37.VoiceFork.OboeThorn.Deliver}]}
    }
   }
  ]
 }
 Speech_C_18 {
  id=#19
  class=Speech_C
  outputs [#20]
  properties {
   softTaskIcon=&T_UI_NPC_MainQuest
   generalConfig {defaultMessage=The gang owes you one., defaultMetadataParameters [RageDebris {value=""}]}
  }
 }
 Speech_C_19 {
  id=#20
  class=Speech_C
  outputs [#21]
  properties {
   softTaskIcon=&T_UI_NPC_MainQuest
   generalConfig {defaultMessage="One ride, that is. To wherever you wanna go. In our newly fixed van!", defaultMetadataParameters [RageDebris {value=""}]}
  }
 }
 ReturnToStart_C_4 {id=#21, class=ReturnToStart_C}
 ConversationEntryPointNode_10 {
  id=#4
  class=ConversationEntryPointNode
  outputs [#22, #23]
  properties {entryTag {tagName=Conversation.Entry.Season37.VoiceFork.OboeThorn}}
 }
 ShowUserDataDrivenCannotUseReason_C_1 {
  id=#23
  class=ShowUserDataDrivenCannotUseReason_C
  outputs []
  properties {
   cannotUseReason=JobAlreadyActive
   serviceBriefCollectionKeyOveride=Service.JobBoard.DisplayCannotUseReason
   serviceBriefConfigCollection=&FoundQuestsServiceBriefCollectionConfig
   softTaskIcon=&Cancel_Normal
  }
  sub_nodes [
   ConversationChoiceNode_9 {
    class=ConversationChoiceNode
    properties {defaultChoiceDisplayText=MISSION ALREADY ACTIVE, choiceTags {gameplayTags [{tagName=Conversation.Choice.Category.Back}]}}
   }
   FortConversationRequirement_ControllerRequirement_13 {
    class=FortConversationRequirement_ControllerRequirement
    properties {nodeName=Require: Has Urgent Quest, requirement {class=FortControllerRequirement_HasUrgentQuest}}
   }
   FortConversationRequirement_ControllerRequirement_5 {
    class=FortConversationRequirement_ControllerRequirement
    properties {
     nodeName=Require: Has Urgent Quest
     requirement {
      class=FortControllerRequirement_HasUrgentQuest
      properties {bInverseCondition=true, acceptedUrgentEventTags {gameplayTags [{tagName=SpecialEvent.Season37.Found.VoiceFork.OboeThorn}]}}
     }
    }
   }
  ]
 }
 Speech_C_9 {
  id=#22
  class=Speech_C
  outputs [#24, #25]
  properties {
   softTaskIcon=&T_UI_NPC_MainQuest
   generalConfig {defaultMessage="The Mystery Machine broke down... again!", defaultMetadataParameters [RageDebris {value=""}]}
  }
  sub_nodes [
   ConversationChoiceNode_27 {
    class=ConversationChoiceNode
    properties {defaultChoiceDisplayText=Need help?, choiceTags {gameplayTags [{tagName=Conversation.Choice.Category.MainQuestline}]}}
   }
   FortConversationRequirement_ControllerRequirement_33 {
    class=FortConversationRequirement_ControllerRequirement
    properties {nodeName=Require: Has Urgent Quest, requirement {class=FortControllerRequirement_HasUrgentQuest, properties {bInverseCondition=true}}}
   }
   FortConversationRequirement_ControllerRequirement_34 {
    class=FortConversationRequirement_ControllerRequirement
    properties {
     nodeName=Require: Has Completed Quest Objective
     requirement {
      class=FortControllerRequirement_HasCompletedQuestObjective
      properties {
       bInverseCondition=true
       objectiveBackendName=Quest_S37_VoiceFork_PermaFrost_Q01_obj0
       questDefinition=&Quest_S37_VoiceFork_PermaFrost_Q01
      }
     }
    }
   }
  ]
 }
 Speech_C_17 {
  id=#25
  class=Speech_C
  outputs [#26]
  properties {
   softTaskIcon=&T_UI_NPC_Back
   generalConfig {defaultMessage=I've fixed 'er before. I can do it again., defaultMetadataParameters [RageDebris {value=""}]}
  }
  sub_nodes [ConversationChoiceNode_8 {
    class=ConversationChoiceNode
    properties {defaultChoiceDisplayText="Nah, not now", choiceTags {gameplayTags [{tagName=Conversation.Choice.CustomSortPriority.Highest}]}}
   }]
 }
 ReturnToStart_C_3 {id=#26, class=ReturnToStart_C}
 GrantStaticQuest_C_5 {
  id=#24
  class=GrantStaticQuest_C
  outputs [#27]
  properties {
   bAssignQuestOwner=true
   bSharedQuestLimitsEnforced=false
   bShouldOverrideSpeakerText=false
   nodeName=Grant Quest
   questToGrantPtr=&Quest_S37_VoiceFork_PermaFrost_OboeThorn_Q01
  }
  sub_nodes [ConversationChoiceNode_26 {
    class=ConversationChoiceNode
    properties {defaultChoiceDisplayText="Yes, count on me!", choiceTags {gameplayTags [{tagName=Conversation.Choice.CustomSortPriority.Lowest}]}}
   }]
 }
 Speech_C_16 {
  id=#27
  class=Speech_C
  outputs []
  properties {
   softTaskIcon=&T_UI_NPC_MainQuest
   generalConfig {defaultMessage="Let's split up and look for parts!", defaultMetadataParameters [RageDebris {value=""}]}
  }
 }
 ConversationEntryPointNode_2 {
  id=#3
  class=ConversationEntryPointNode
  outputs [#28, #29]
  properties {entryTag {tagName=Conversation.Entry.Season37.VoiceFork.ArcanaAgate.DustyBun}}
 }
 Speech_C_0 {
  id=#29
  class=Speech_C
  outputs [#30]
  properties {
   softTaskIcon=&T_UI_NPC_MainQuest
   generalConfig {defaultMessage=Huh?, defaultMetadataParameters [RageDebris {value=""}]}
  }
  sub_nodes [
   ConversationChoiceNode_3 {
    class=ConversationChoiceNode
    properties {defaultChoiceDisplayText=Interrogate, choiceTags {gameplayTags [{tagName=Conversation.Choice.Category.MainQuestline}]}}
   }
   FortConversationRequirement_ControllerRequirement_4 {
    class=FortConversationRequirement_ControllerRequirement
    properties {
     nodeName=Require: Has GameplayTag
     requirement {class=FortControllerRequirement_HasOwnedTag, properties {ownedTag {tagName=CustomVerb.Quest.S37.VoiceFork.PermaFrost.ArcanaAgate.Failure}}}
    }
   }
   FortConversationRequirement_ControllerRequirement_22 {
    class=FortConversationRequirement_ControllerRequirement
    properties {
     nodeName=Require: Has Completed Quest Objective
     requirement {
      class=FortControllerRequirement_HasCompletedQuestObjective
      properties {
       bInverseCondition=true
       objectiveBackendName=Quest_S37_VoiceFork_PermaFrost_ArcanaAgate_Q01_obj0
       questDefinition=/VoiceFork_BR_Quests/Items/QuestItems/PermaFrost/UrgentQuests/ArcanaAgate/Quest_S37_VoiceFork_PermaFrost_ArcanaAgate_Q01
      }
     }
    }
   }
  ]
 }
 Speech_C_3 {
  id=#30
  class=Speech_C
  outputs [#31]
  properties {
   softTaskIcon=&T_UI_NPC_MainQuest
   generalConfig {defaultMessage=I don't know what you're talking about..., defaultMetadataParameters [RageDebris {value=""}]}
  }
 }
 ReturnToStart_C_7 {id=#31, class=ReturnToStart_C}
 Speech_C_2 {
  id=#28
  class=Speech_C
  outputs [#32]
  properties {
   softTaskIcon=&T_UI_NPC_MainQuest
   generalConfig {defaultMessage=I know stealing is wrong..., defaultMetadataParameters [RageDebris {value=""}]}
  }
  sub_nodes [
   ConversationChoiceNode_15 {
    class=ConversationChoiceNode
    properties {
     defaultChoiceDisplayText=Interrogate
     choiceTags {gameplayTags [{tagName=Conversation.Choice.Category.MainQuestline}, {tagName=Conversation.Entry.Season37.VoiceFork.ArcanaAgate.DustyBun}]}
    }
   }
   FortConversationRequirement_ControllerRequirement_16 {
    class=FortConversationRequirement_ControllerRequirement
    properties {
     nodeName=Require: Has Urgent Quest
     requirement {
      class=FortControllerRequirement_HasUrgentQuest
      properties {acceptedUrgentEventTags {gameplayTags [{tagName=SpecialEvent.Season37.Found.VoiceFork.ArcanaAgate}]}}
     }
    }
   }
   FortConversationRequirement_ControllerRequirement_19 {
    class=FortConversationRequirement_ControllerRequirement
    properties {
     nodeName=Require: Has Completed Quest Objective
     requirement {
      class=FortControllerRequirement_HasCompletedQuestObjective
      properties {
       bInverseCondition=true
       objectiveBackendName=Quest_S37_VoiceFork_PermaFrost_Q01_obj2
       questDefinition=&Quest_S37_VoiceFork_PermaFrost_Q01
      }
     }
    }
   }
   FortConversationRequirement_ControllerRequirement_20 {
    class=FortConversationRequirement_ControllerRequirement
    properties {
     nodeName=Require: Has Completed Quest Objective
     requirement {
      class=FortControllerRequirement_HasCompletedQuestObjective
      properties {
       objectiveBackendName=Quest_S37_VoiceFork_PermaFrost_ArcanaAgate_Q01_obj0
       questDefinition=&Quest_S37_VoiceFork_PermaFrost_ArcanaAgate_Q01
      }
     }
    }
   }
  ]
 }
 Speech_C_12 {
  id=#32
  class=Speech_C
  outputs [#33]
  properties {
   softTaskIcon=&T_UI_NPC_MainQuest
   generalConfig {
    defaultMessage="It's just... for once, I wanted to *solve* a mystery instead of *being* the mystery."
    defaultMetadataParameters [RageDebris {value=""}]
   }
  }
 }
 Speech_C_14 {
  id=#33
  class=Speech_C
  outputs [#34]
  properties {
   softTaskIcon=&T_UI_NPC_MainQuest
   generalConfig {defaultMessage="I thought if I dressed the part, the gang would let me join them.", defaultMetadataParameters [RageDebris {value=""}]}
  }
 }
 Speech_C_15 {
  id=#34
  class=Speech_C
  outputs [#35]
  properties {
   softTaskIcon=&T_UI_NPC_MainQuest
   generalConfig {
    defaultMessage=But now I see... my lot in life is to hang out on bridges and foretell disaster. Sigh...
    defaultMetadataParameters [RageDebris {value=""}]
   }
  }
 }
 ReturnToStart_C_2 {id=#35, class=ReturnToStart_C}
 ConversationEntryPointNode_3 {
  id=#2
  class=ConversationEntryPointNode
  outputs [#36, #37]
  properties {entryTag {tagName=Conversation.Entry.Season37.VoiceFork.ArcanaAgate}}
 }
 ShowUserDataDrivenCannotUseReason_C_0 {
  id=#37
  class=ShowUserDataDrivenCannotUseReason_C
  outputs []
  properties {
   cannotUseReason=JobAlreadyActive
   serviceBriefCollectionKeyOveride=Service.JobBoard.DisplayCannotUseReason
   serviceBriefConfigCollection=&FoundQuestsServiceBriefCollectionConfig
   softTaskIcon=&Cancel_Normal
  }
  sub_nodes [
   ConversationChoiceNode_0 {
    class=ConversationChoiceNode
    properties {defaultChoiceDisplayText=MISSION ALREADY ACTIVE, choiceTags {gameplayTags [{tagName=Conversation.Choice.Category.Back}]}}
   }
   FortConversationRequirement_ControllerRequirement_1 {
    class=FortConversationRequirement_ControllerRequirement
    properties {nodeName=Require: Has Urgent Quest, requirement {class=FortControllerRequirement_HasUrgentQuest}}
   }
   FortConversationRequirement_ControllerRequirement_3 {
    class=FortConversationRequirement_ControllerRequirement
    properties {
     nodeName=Require: Has Urgent Quest
     requirement {
      class=FortControllerRequirement_HasUrgentQuest
      properties {bInverseCondition=true, acceptedUrgentEventTags {gameplayTags [{tagName=SpecialEvent.Season37.Found.VoiceFork.ArcanaAgate}]}}
     }
    }
   }
  ]
 }
 Speech_C_1 {
  id=#36
  class=Speech_C
  outputs [#38, #39]
  properties {
   softTaskIcon=&T_UI_NPC_MainQuest
   generalConfig {
    defaultMessage="Some creep stole my whole wardrobe! Ugh, I can't think clearly wearing last season's trends..."
    defaultMetadataParameters [RageDebris {value=""}]
   }
  }
  sub_nodes [
   ConversationChoiceNode_13 {
    class=ConversationChoiceNode
    properties {defaultChoiceDisplayText=Need help?, choiceTags {gameplayTags [{tagName=Conversation.Choice.Category.MainQuestline}]}}
   }
   FortConversationRequirement_ControllerRequirement_11 {
    class=FortConversationRequirement_ControllerRequirement
    properties {nodeName=Require: Has Urgent Quest, requirement {class=FortControllerRequirement_HasUrgentQuest, properties {bInverseCondition=true}}}
   }
   FortConversationRequirement_ControllerRequirement_12 {
    class=FortConversationRequirement_ControllerRequirement
    properties {
     nodeName=Require: Has Completed Quest Objective
     requirement {
      class=FortControllerRequirement_HasCompletedQuestObjective
      properties {
       bInverseCondition=true
       objectiveBackendName=Quest_S37_VoiceFork_PermaFrost_Q01_obj2
       questDefinition=&Quest_S37_VoiceFork_PermaFrost_Q01
      }
     }
    }
   }
  ]
 }
 Speech_C_11 {
  id=#39
  class=Speech_C
  outputs [#40]
  properties {
   softTaskIcon=&T_UI_NPC_Back
   generalConfig {defaultMessage="Maybe scarves will make a comeback really, really quickly...", defaultMetadataParameters [RageDebris {value=""}]}
  }
  sub_nodes [ConversationChoiceNode_7 {
    class=ConversationChoiceNode
    properties {defaultChoiceDisplayText="Nah, not now", choiceTags {gameplayTags [{tagName=Conversation.Choice.CustomSortPriority.Highest}]}}
   }]
 }
 ReturnToStart_C_1 {id=#40, class=ReturnToStart_C}
 GrantStaticQuest_C_2 {
  id=#38
  class=GrantStaticQuest_C
  outputs [#41]
  properties {
   bAssignQuestOwner=true
   bSharedQuestLimitsEnforced=false
   bShouldOverrideSpeakerText=false
   nodeName=Grant Quest
   questToGrantPtr=&Quest_S37_VoiceFork_PermaFrost_ArcanaAgate_Q01
  }
  sub_nodes [ConversationChoiceNode_14 {
    class=ConversationChoiceNode
    properties {defaultChoiceDisplayText="Yes, count on me!", choiceTags {gameplayTags [{tagName=Conversation.Choice.CustomSortPriority.Lowest}]}}
   }]
 }
 Speech_C_8 {
  id=#41
  class=Speech_C
  outputs []
  properties {
   softTaskIcon=&T_UI_NPC_MainQuest
   generalConfig {defaultMessage=My headband collection thanks you., defaultMetadataParameters [RageDebris {value=""}]}
  }
 }
 ConversationEntryPointNode_1 {
  id=#1
  class=ConversationEntryPointNode
  outputs [#42, #43]
  properties {entryTag {tagName=Conversation.Entry.Season37.VoiceFork.BaskIsle}}
 }
 ShowUserDataDrivenCannotUseReason_C_2 {
  id=#43
  class=ShowUserDataDrivenCannotUseReason_C
  outputs []
  properties {
   cannotUseReason=JobAlreadyActive
   serviceBriefCollectionKeyOveride=Service.JobBoard.DisplayCannotUseReason
   serviceBriefConfigCollection=&FoundQuestsServiceBriefCollectionConfig
   softTaskIcon=&Cancel_Normal
  }
  sub_nodes [
   ConversationChoiceNode_4 {
    class=ConversationChoiceNode
    properties {defaultChoiceDisplayText=MISSION ALREADY ACTIVE, choiceTags {gameplayTags [{tagName=Conversation.Choice.Category.Back}]}}
   }
   FortConversationRequirement_ControllerRequirement_10 {
    class=FortConversationRequirement_ControllerRequirement
    properties {nodeName=Require: Has Urgent Quest, requirement {class=FortControllerRequirement_HasUrgentQuest}}
   }
   FortConversationRequirement_ControllerRequirement_0 {
    class=FortConversationRequirement_ControllerRequirement
    properties {
     nodeName=Require: Has Urgent Quest
     requirement {
      class=FortControllerRequirement_HasUrgentQuest
      properties {bInverseCondition=true, acceptedUrgentEventTags {gameplayTags [{tagName=SpecialEvent.Season37.Found.VoiceFork.BaskIsle}]}}
     }
    }
   }
  ]
 }
 Speech_C_10 {
  id=#42
  class=Speech_C
  outputs [#44, #45]
  properties {
   softTaskIcon=&T_UI_NPC_MainQuest
   generalConfig {defaultMessage="I dropped my lucky magnifying glass. I can't solve mysteries without it!", defaultMetadataParameters [RageDebris {value=""}]}
  }
  sub_nodes [
   ConversationChoiceNode_5 {
    class=ConversationChoiceNode
    properties {defaultChoiceDisplayText=Need help?, choiceTags {gameplayTags [{tagName=Conversation.Choice.Category.MainQuestline}]}}
   }
   FortConversationRequirement_ControllerRequirement_6 {
    class=FortConversationRequirement_ControllerRequirement
    properties {
     nodeName=Require: Has Completed Quest Objective
     requirement {
      class=FortControllerRequirement_HasCompletedQuestObjective
      properties {
       bInverseCondition=true
       objectiveBackendName=Quest_S37_VoiceFork_PermaFrost_Q01_obj1
       questDefinition=&Quest_S37_VoiceFork_PermaFrost_Q01
      }
     }
    }
   }
   FortConversationRequirement_ControllerRequirement_2 {
    class=FortConversationRequirement_ControllerRequirement
    properties {nodeName=Require: Has Urgent Quest, requirement {class=FortControllerRequirement_HasUrgentQuest, properties {bInverseCondition=true}}}
   }
  ]
 }
 Speech_C_7 {
  id=#45
  class=Speech_C
  outputs [#46]
  properties {
   softTaskIcon=&T_UI_NPC_Back
   generalConfig {defaultMessage=Wouldn't be the first time I solved a mystery by myself., defaultMetadataParameters [RageDebris {value=""}]}
  }
  sub_nodes [ConversationChoiceNode_6 {
    class=ConversationChoiceNode
    properties {defaultChoiceDisplayText="Nah, not now", choiceTags {gameplayTags [{tagName=Conversation.Choice.CustomSortPriority.Highest}]}}
   }]
 }
 ReturnToStart_C_0 {id=#46, class=ReturnToStart_C}
 GrantStaticQuest_C_0 {
  id=#44
  class=GrantStaticQuest_C
  outputs [#47]
  properties {
   bAssignQuestOwner=true
   bSharedQuestLimitsEnforced=false
   bShouldOverrideSpeakerText=false
   nodeName=Grant Quest
   questToGrantPtr=/VoiceFork_BR_Quests/Items/QuestItems/PermaFrost/UrgentQuests/BaskIsle/Quest_S37_VoiceFork_PermaFrost_BaskIsle_Q01
  }
  sub_nodes [ConversationChoiceNode_1 {
    class=ConversationChoiceNode
    properties {defaultChoiceDisplayText="Yes, count on me!", choiceTags {gameplayTags [{tagName=Conversation.Choice.CustomSortPriority.Lowest}]}}
   }]
 }
 Speech_C_6 {
  id=#47
  class=Speech_C
  outputs []
  properties {
   softTaskIcon=&T_UI_NPC_MainQuest
   generalConfig {
    defaultMessage=I'll make sure you get a junior co-sleuth credit in my autobiography someday.
    defaultMetadataParameters [RageDebris {value=""}]
   }
  }
 }
]
